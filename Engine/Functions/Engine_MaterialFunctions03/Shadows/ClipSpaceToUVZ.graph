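# Material function in the engine's shadow utilities that converts clip and local positions into UV coordinates and a linear depth.
# UV comes from a perspective divide of clip-space XY by W, with W clamped to at least the optional NearClip input, then scaled and offset into UV space with Y flipped.
# Depth is simply the Z component of the local position input.
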MaterialFunction /Engine/Functions/Engine_MaterialFunctions03/Shadows/ClipSpaceToUVZ
graph {
 properties {description=Utility to convert local and clip positions to UV / Depth}
 nodes [
  n1 = Constant2Vector() {g=-0.5, r=0.5}
  n2 = FunctionInput() {inputName=ClipPos, inputType=FunctionInput_Vector4}
  n3 = FunctionInput() {bUsePreviewValueAsDefault=true, inputName=NearClip, inputType=FunctionInput_Scalar, previewValue=<1 0 0 1>}
  n4 = Divide(A=ComponentMask(None=@n2) {g=true, r=true}, B=Max(A=ComponentMask(None=@n2) {a=true}, B=@n3)) {desc=Perspective}
  FunctionOutput(None=Add(A=Constant() {r=0.5}, B=Multiply(A=@n1, B=@n4) {constB=0.5}) {constB=0.5, desc=UV Space}) {outputName=UV}
  FunctionOutput(None=ComponentMask(None=FunctionInput() {inputName=LocalPos}) {b=true, desc=Linear Depth}) {outputName=Depth}
 ]
}
