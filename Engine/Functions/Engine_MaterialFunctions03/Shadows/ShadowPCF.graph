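# Material function that prepares PCF shadow-map sampling, outputting gather UV, interpolation weights, slope bias, and backface and clip masks.
# UV snaps to the texel grid at a half-texel offset, with the remainder as Lerp.
# A shadow-space eye vector, perspective or orthographic, meets the primitive plane for per-texel SlopeBias, with clamped N·Eye avoiding division by zero.
# Backface uses the vertex normal; Clip flags points outside the side or near planes.
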MaterialFunction /Engine/Functions/Engine_MaterialFunctions03/Shadows/ShadowPCF
graph {
 properties {description="Computes adjusted UV for gather, UV interpolation, and SlopeBias for PCF"}
 nodes [
  Constant()
  n1 = FunctionInput(Preview=TextureObject()) {inputName=Texture, inputType=FunctionInput_Texture2D}
  n2 = Add(
   A=Multiply(A=FunctionInput() {inputName=UV, inputType=FunctionInput_Vector2}, B=TextureProperty(Tex=@n1))
   B=Constant2Vector() {g=0.5, r=0.5}
  ) {constB=0}
  n3 = Floor(None=@n2)
  n4 = TextureProperty(Tex=@n1) {property=TMTM_TexelSize}
  FunctionOutput(None=Multiply(A=@n3, B=@n4)) {outputName=AdjustedUV}
  FunctionOutput(None=Subtract(A=@n2, B=@n3)) {outputName=Lerp}
  n5 = FunctionInput() {bUsePreviewValueAsDefault=true, description=In shadow space, inputName=PrimitiveNormal, previewValue=<0 0 1 1>}
  n6 = FunctionInput() {description=In shadow space, inputName=LocalPos}
  n7 = ComponentMask(None=@n6) {g=true, r=true}
  n8 = FunctionInput() {bUsePreviewValueAsDefault=true, description=Projection applied to 110 vector, inputName=Projection110, inputType=FunctionInput_Vector4}
  n9 = OneMinus(None=ComponentMask(None=@n8) {a=true, desc="0=persp 1=ortho"}) {desc="1=persp 0=ortho"}
  n10 = FunctionInput() {bUsePreviewValueAsDefault=true, description=In shadow space, inputName=ClipPos, inputType=FunctionInput_Vector4}
  n11 = Normalize(VectorInput=AppendVector(A=Multiply(A=@n7, B=@n9) {desc="Eye vector 00 for ortho, xy for perspective"}, B=ComponentMask(None=@n10) {a=true})) {desc=Shadow eye vector}
  n12 = Multiply(
   A=Multiply(A=@n4, B=ComponentMask(None=@n5) {g=true, r=true})
   B=Multiply(A=ComponentMask(None=@n11) {b=true}, B=ComponentMask(None=@n10) {a=true})
  )
  n13 = Constant2Vector() {g=2, r=-2}
  n14 = ComponentMask(None=@n8) {desc=Projection scale, g=true, r=true}
  n15 = FunctionInput() {bUsePreviewValueAsDefault=true, inputName=NormalClamp, inputType=FunctionInput_Scalar, previewValue=<0.03 0 0 1>}
  n16 = Subtract(B=@n15) {constA=0}
  n17 = Min(A=DotProduct(A=@n11, B=@n5), B=@n16) {desc=Eye dot N clamp to avoid div by 0}
  FunctionOutput(None=Divide(A=Multiply(A=@n12, B=@n13), B=Multiply(A=@n14, B=@n17))) {
   desc="Formula:\r\n2.0 * TexelSize.xy * Normal.xy * Eye.z * ClipPos.w\r\n------------------------------------------------\r\ndot( Eye, Normal ) * ProjectionScale.xy\r\n\r\nDerived by subtracting intersection of Eye ray from\r\nEye ray + TexelSize, with primitive plane defined\r\nby ClipPos and Normal.  Then scaling by 2/ProjectionScale\r\n(the 2 comes from the fact that ProjectionScale is for half theProjectionScale is for the\r\nshadow map transform, not the view, and Eye and\r\nClipPos are both in local shadow map space."
   outputName=SlopeBias
  }
  n18 = FunctionInput() {bUsePreviewValueAsDefault=true, description=In shadow space, inputName=VertexNormal, previewValue=<0 0 1 1>}
  n19 = Subtract(B=DotProduct(A=@n11, B=@n18)) {constA=0}
  FunctionOutput(None=Saturate(None=Multiply(A=@n19) {constB=5})) {description="Zero if output is backface, based on vertex normal", outputName=Backface}
  n20 = Abs(None=ComponentMask(None=@n10) {r=true})
  n21 = Abs(None=ComponentMask(None=@n10) {g=true})
  n22 = ComponentMask(None=@n6) {b=true}
  n23 = FunctionInput() {bUsePreviewValueAsDefault=true, inputName=NearClip, inputType=FunctionInput_Scalar, previewValue=<1 0 0 1>}
  n24 = Constant()
  n25 = Constant() {r=1}
  FunctionOutput(None=If(
    A=Max(A=@n20, B=@n21)
    B=ComponentMask(None=@n10) {a=true}
    "A > B"=Constant() {r=1}
    "A < B"=If(A=@n22, B=@n23, "A > B"=@n24, "A < B"=@n25) {desc=Near clip}
   ) {desc=Side planes clip}) {description=One if output is outside frustum, outputName=Clip}
  If(A=@n17, B=@n16, "A > B"=Constant(), "A < B"=Constant() {r=1}) {desc=Should B be normal clamp or zero?}
 ]
}
