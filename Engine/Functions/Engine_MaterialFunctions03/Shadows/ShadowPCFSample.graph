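# Material function for one 2x2 tap of percentage-closer-filtered shadows, returning a weighted sum of depth comparisons.
# A red-channel texture gather fetches four texels at a texel-scaled offset from AdjustedUV.
# Each is compared against Depth plus a per-texel slope bias, clamped to never push forward.
# Passing texels count fully, except along the kernel's outer edges set by Size, where Lerp supplies bilinear weights.
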MaterialFunction /Engine/Functions/Engine_MaterialFunctions03/Shadows/ShadowPCFSample
graph {
 properties {description="Samples a 2x2 set of pixels using gather, and applies bilinear weighting to the result"}
 nodes [
  n1 = FunctionInput() {inputName=Offset, inputType=FunctionInput_Vector2}
  ComponentMask(None=@n1) {r=true}
  ComponentMask(None=@n1) {g=true}
  n2 = FunctionInput() {inputName=Lerp, inputType=FunctionInput_Vector2}
  n3 = FunctionInput() {inputName=SlopeBias, inputType=FunctionInput_Vector2}
  n4 = Multiply(A=Subtract(A=@n1, B=@n2) {constA=0, desc=Offset Minus Lerp}, B=@n3)
  n5 = ComponentMask(None=@n4) {r=true}
  n6 = Multiply(A=Add(A=@n1, B=OneMinus(None=@n2)) {desc=Offset Plus One Minus Lerp}, B=@n3)
  n7 = ComponentMask(None=@n6) {g=true}
  n8 = Add(A=@n5, B=@n7)
  n9 = FunctionInput() {inputName=Depth, inputType=FunctionInput_Scalar}
  n10 = ComponentMask(None=@n6) {r=true}
  n11 = Add(A=@n10, B=@n7)
  n12 = ComponentMask(None=@n4) {g=true}
  n13 = Add(A=@n10, B=@n12)
  n14 = Add(A=@n5, B=@n12)
  n15 = MaterialFunctionCall(
   X=Add(A=Min(A=@n8) {constB=0, desc=Prevent forward bias}, B=@n9) {desc=Slope Bias}
   Y=Add(A=Min(A=@n11) {constB=0}, B=@n9)
   Z=Add(A=Min(A=@n13) {constB=0}, B=@n9)
   A=Add(A=Min(A=@n14) {constB=0}, B=@n9)
  ) {materialFunction=&MakeFloat4}
  n16 = FunctionInput(Preview=TextureObject()) {inputType=FunctionInput_Texture2D}
  n17 = TextureProperty(Tex=@n16) {property=TMTM_TexelSize}
  n18 = TextureSample(UVs=Add(A=FunctionInput() {inputName=AdjustedUV, inputType=FunctionInput_Vector2}, B=Multiply(A=@n1, B=@n17)), Tex=@n16) {gatherMode=TGM_Red, samplerSource=SSM_Clamp_WorldGroupSettings, samplerType=SAMPLERTYPE_LinearColor}
  n19 = Constant4Vector()
  n20 = Subtract(A=FunctionInput() {inputName=Size, inputType=FunctionInput_Scalar})
  n21 = ComponentMask(None=@n1) {g=true}
  n22 = Constant() {r=1}
  n23 = ComponentMask(None=@n2) {g=true}
  n24 = If(A=@n20, B=@n21, "A > B"=@n22, "A == B"=@n23, "A < B"=@n22) {desc=Bottom weight}
  n25 = ComponentMask(None=@n1) {r=true}
  n26 = Constant() {r=1}
  n27 = ComponentMask(None=@n2) {r=true}
  n28 = If(A=Subtract(B=@n20) {constA=0}, B=@n25, "A > B"=@n26, "A == B"=OneMinus(None=@n27), "A < B"=@n26) {desc=Left weight}
  n29 = If(A=@n20, B=@n25, "A > B"=@n26, "A == B"=@n27, "A < B"=@n26) {desc=Right weight}
  n30 = If(A=Subtract(B=@n20) {constA=0}, B=@n21, "A > B"=@n22, "A == B"=OneMinus(None=@n23), "A < B"=@n22) {desc=Top weight}
  n31 = MaterialFunctionCall(X=Multiply(A=@n24, B=@n28), Y=Multiply(A=@n24, B=@n29), Z=Multiply(A=@n30, B=@n29), A=Multiply(A=@n30, B=@n28)) {materialFunction=&MakeFloat4}
  FunctionOutput(None=MaterialFunctionCall(f4=If(A=@n15.Result, B=@n18.RGBA, "A > B"=@n19.RGBA, "A < B"=@n31.Result)) {materialFunction=&AddComponents}.f4)
 ]
}
