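# Material function that computes a 4x4 percentage-closer-filtered shadow term from a shadow map texture and shadow-space inputs.
# ClipSpaceToUVZ yields UV and depth; ShadowPCF derives sampling, backface and clip terms from these UVs, positions and normals.
# Four ShadowPCFSample calls at diagonal offsets compare against constant-biased depth, and their sum is scaled by one ninth.
# A Min applies backface shadowing, then a Max applies frustum clipping.
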MaterialFunction /Engine/Functions/Engine_MaterialFunctions03/Shadows/ShadowPCF4x4
graph {nodes [
  n1 = FunctionInput() {bUsePreviewValueAsDefault=true, inputName=BackfaceTolerance, inputType=FunctionInput_Scalar, previewValue=<0.02 0 0 1>}
  n2 = FunctionInput() {bUsePreviewValueAsDefault=true, inputName=NearClip, inputType=FunctionInput_Scalar, previewValue=<1 0 0 1>}
  n3 = FunctionInput() {description="Shadow space position (View * WorldPos)", inputName=LocalPos}
  n4 = FunctionInput() {description="Clip space position (Shadow Projection * LocalPos)", inputName=ClipPos, inputType=FunctionInput_Vector4}
  n5 = MaterialFunctionCall!(NearClip=@n2, LocalPos=@n3, ClipPos=@n4) {materialFunction=&/ClipSpaceToUVZ}
  n6 = FunctionInput() {
   bUsePreviewValueAsDefault=true
   description="Shadow space primitive normal (for slope bias, optional)"
   inputName=LocalPrimitiveNormal
   previewValue=<0 0 1 1>
  }
  n7 = FunctionInput(Preview=TextureObject()) {inputName=Texture, inputType=FunctionInput_Texture2D}
  n8 = FunctionInput() {bUsePreviewValueAsDefault=true, description=Shadow space vertex normal, inputName=LocalVertexNormal, previewValue=<0 0 1 1>}
  n9 = MaterialFunctionCall!(
   NormalClamp=@n1
   UV=@n5.UV
   PrimitiveNormal=@n6
   Projection110=FunctionInput() {description=Projection matrix applied to 110 vector, inputName=Projection110, inputType=FunctionInput_Vector4}
   Texture=@n7
   ClipPos=@n4
   LocalPos=@n3
   NearClip=@n2
   VertexNormal=@n8
  ) {materialFunction=&/ShadowPCF}
  n10 = Constant() {r=2}
  n11 = FunctionInput() {bUsePreviewValueAsDefault=true, inputName=ConstantBias, inputType=FunctionInput_Scalar, previewValue=<0.5 0 0 1>}
  n12 = Subtract(A=@n5.Depth, B=@n11) {constB=0.5, desc=Constant bias}
  n13 = MaterialFunctionCall!(In=@n7, Offset=Constant2Vector() {g=-1, r=-1}, Size=@n10, Lerp=@n9.Lerp, Depth=@n12, SlopeBias=@n9.Lerp, AdjustedUV=@n9.Lerp) {materialFunction=&/ShadowPCFSample}
  n14 = MaterialFunctionCall(In=@n7, Offset=Constant2Vector() {g=-1, r=1}, Size=@n10, Lerp=@n9.Lerp, Depth=@n12, SlopeBias=@n9.Lerp, AdjustedUV=@n9.Lerp) {materialFunction=&ShadowPCFSample}
  n15 = MaterialFunctionCall(In=@n7, Offset=Constant2Vector() {g=1, r=-1}, Size=@n10, Lerp=@n9.Lerp, Depth=@n12, SlopeBias=@n9.Lerp, AdjustedUV=@n9.Lerp) {materialFunction=&ShadowPCFSample}
  n16 = MaterialFunctionCall(In=@n7, Offset=Constant2Vector() {g=1, r=1}, Size=@n10, Lerp=@n9.Lerp, Depth=@n12, SlopeBias=@n9.Lerp, AdjustedUV=@n9.Lerp) {materialFunction=&ShadowPCFSample}
  n17 = Multiply(A=Add(A=Add(A=@n13.Result, B=@n14.Result), B=Add(A=@n15.Result, B=@n16.Result))) {constB=0.1111}
  FunctionOutput(None=Max(A=Min(A=@n9.Backface, B=@n17) {desc=Shadow backfaces}, B=@n9.Clip) {desc=Frustum Clip})
 ]}
